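# Looping outdoor morning ambience cue for the CityLights POI, blending small-town day and desert valley morning wind beds.
# A quality switch sends the first two levels to the quad waves and the last two to the stereo waves.
# Each pair is mixed, then passes through two nested, endlessly looping envelopers: the inner curves pitch, the outer gently swells volume.
# It uses outdoor ambient attenuation and sound class, with a single-instance, prevent-new concurrency override.
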SoundCue /Hera/Audio/POI/CityLights/Ambience_CityLights_Morning_Loop_Ext_Cue
root=QualityLevel_0
nodes [
 {
  id=Enveloper_0
  children [Mixer_1]
  properties {
   bLoop=true
   bLoopIndefinitely=true
   loopEnd=30
   pitchCurve {editorCurveData {keys [{interpMode=RCIM_Cubic, time=0}, <at 15.9468 0.9181 RCIM_Cubic>, {interpMode=RCIM_Cubic, time=30}]}}
   volumeCurve {editorCurveData {keys [<at 0 1>]}}
  }
 }
 {
  id=Enveloper_1
  children [Mixer_0]
  properties {
   bLoop=true
   bLoopIndefinitely=true
   loopEnd=30
   pitchCurve {editorCurveData {keys [{interpMode=RCIM_Cubic, time=0}, <at 15.9468 0.9181 RCIM_Cubic>, {interpMode=RCIM_Cubic, time=30}]}}
   volumeCurve {editorCurveData {keys [<at 0 1>]}}
  }
 }
 {
  id=Enveloper_2
  children [Enveloper_1]
  properties {
   bLoop=true
   bLoopIndefinitely=true
   loopEnd=20
   pitchCurve {editorCurveData {keys [{time=0}]}}
   volumeCurve {editorCurveData {keys [
      {arriveTangent=-0.01, interpMode=RCIM_Cubic, time=0, value=1}
      <at 10 1.1 RCIM_Cubic>
      {interpMode=RCIM_Cubic, leaveTangent=0.01, time=20, value=1}
     ]}}
  }
 }
 {
  id=Enveloper_3
  children [Enveloper_0]
  properties {
   bLoop=true
   bLoopIndefinitely=true
   loopEnd=30
   pitchCurve {editorCurveData {keys [{time=0}]}}
   volumeCurve {editorCurveData {keys [
      {arriveTangent=-0.01, interpMode=RCIM_Cubic, time=0, value=1}
      <at 10 1.1 RCIM_Cubic>
      {interpMode=RCIM_Cubic, leaveTangent=0.01, time=20, value=1}
     ]}}
  }
 }
 {id=Mixer_0, children [WavePlayer_3,WavePlayer_4], properties {inputVolume [0.551,0.8]}}
 {id=Mixer_1, children [WavePlayer_5,WavePlayer_1], properties {inputVolume [0.4,0.8]}}
 {id=QualityLevel_0, children [Enveloper_2,Enveloper_2,Enveloper_3,Enveloper_3]}
 {id=WavePlayer_1, properties {bLooping=true, soundWaveAssetPtr=/AmbientAudioContent/Waves/POI/Ambience_Small_Town_Day_Stereo}}
 {id=WavePlayer_3, properties {bLooping=true, soundWaveAssetPtr=/AmbientAudioContent/Waves/POI/Ambience_Small_Town_Day_Quad}}
 {
  id=WavePlayer_4
  properties {bLooping=true, soundWaveAssetPtr=/AmbientAudioContent/Waves/POI/Ambience_Desert_Valley_Wind_Morning_Ext_Loop_Quad}
 }
 {
  id=WavePlayer_5
  properties {bLooping=true, soundWaveAssetPtr=/AmbientAudioContent/Waves/POI/Ambience_Desert_Valley_Wind_Morning_Ext_Loop_Stereo}
 }
]
properties {
 attenuationSettings=/AmbientAudioBase/Attenuations/Fort_Ambient_Att
 bEnableBusSends=false
 bOverrideConcurrency=true
 maxDistance=2097152
 soundClassObject=&Fort_Ambient_Outdoors_Athena
 volumeMultiplier=1
 concurrencyOverrides {maxCount=1, resolutionRule=PreventNew}
 concurrencySet [/AmbientAudioBase/Concurrency/AmbientConc_2Max_Loop_PreventNew_Default_Concurrency]
 preEffectBusSends [{}]
}
